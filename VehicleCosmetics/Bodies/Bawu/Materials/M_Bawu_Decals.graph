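# Masked, Default Lit material for the decals on the Bawu vehicle body cosmetic, enabled for skeletal meshes and mesh deformers.
# It samples three Bawu decal textures: base color, normal, and a linear-color "Textures" map.
# The normal sample is multiplied by a constant vector that doubles X and Y and keeps Z, strengthening the surface detail.
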Material /VehicleCosmetics/Bodies/Bawu/Materials/M_Bawu_Decals
graph {
 properties {bUsedWithMeshDeformer=true, bUsedWithSkeletalMesh=true, blendMode=BLEND_Masked, usedShadingModels=Default Lit}
 nodes [
  Multiply(
   A=TextureSample() {samplerType=SAMPLERTYPE_Normal, texture=/VehicleCosmetics/Bodies/Bawu/Textures/Decals/T_Bawu_Decals_N}.RGB
   B=Constant3Vector() {constant=<rgba 2 2 1 0>}
  )
  TextureSample() {texture=/VehicleCosmetics/Bodies/Bawu/Textures/Decals/T_Bawu_Decals_BC}
  TextureSample() {samplerType=SAMPLERTYPE_LinearColor, texture=/VehicleCosmetics/Bodies/Bawu/Textures/Decals/T_Bawu_Decals_Textures}
 ]
}
