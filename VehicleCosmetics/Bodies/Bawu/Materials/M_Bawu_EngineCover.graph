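# Masked material for the Bawu vehicle body's engine cover, layering a hex grille detail over the cover's own textures.
# The inverted green channel of the cover mask, partly combined with vertex blue, gates where the grille alters cutout, normal, roughness and base color shading.
# Bump-offset UVs give the grille parallax, with offset and un-offset mask samples summed.
# Powered vertex alpha also feeds a height-to-normal border calculation.
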Material /VehicleCosmetics/Bodies/Bawu/Materials/M_Bawu_EngineCover
graph {
 properties {bUsedWithMeshDeformer=true, bUsedWithNanite=true, bUsedWithSkeletalMesh=true, blendMode=BLEND_Masked, usedShadingModels=Default Lit}
 nodes [
  n1 = VertexColor()
  n2 = TextureSample() {constCoordinate=1, samplerType=SAMPLERTYPE_LinearColor, texture=/VehicleCosmetics/Bodies/Bawu/Textures/Decals/T_Bawu_EngineCover_M}
  n3 = OneMinus(None=@n2.G)
  Add(A=@n1.A, B=@n3)
  n4 = TextureCoordinate()
  n5 = TextureSample() {samplerType=SAMPLERTYPE_LinearGrayscale, texture=/VehicleCosmetics/Bodies/Bawu/Textures/Decals/T_Detail_HexGrille_M}
  n6 = BumpOffset(Coordinate=@n4, Height=@n5.R, HeightRatioInput=ScalarParameter() {defaultValue=0.06, parameterName=HeightRatio}) {referencePlane=1}
  n7 = Multiply(A=@n1.B, B=@n3)
  LinearInterpolate(
   B=Add(
    A=TextureSample(UVs=@n6) {samplerType=SAMPLERTYPE_LinearGrayscale, texture=&T_Detail_HexGrille_M}.R
    B=@n5.R
   )
   Alpha=@n7
  ) {constA=1}
  n8 = BumpOffset(
   Coordinate=@n4
   Height=TextureSample() {samplerType=SAMPLERTYPE_LinearGrayscale, texture=/VehicleCosmetics/Bodies/Bawu/Textures/Decals/T_Detail_HexGrille_H}.R
   HeightRatioInput=ScalarParameter() {defaultValue=0.1, parameterName=HeightRatioNormal}
  ) {referencePlane=1}
  LinearInterpolate(
   A=TextureSample() {constCoordinate=1, samplerType=SAMPLERTYPE_Normal, texture=/VehicleCosmetics/Bodies/Bawu/Textures/Decals/T_Bawu_EngineCover_N}.RGB
   B=Multiply(
    A=VectorParameter() {parameterName=BorderNormalIntensity, defaultValue=<rgba 1 1 1 1>}.RGB
    B=TextureSample(UVs=@n8) {samplerType=SAMPLERTYPE_Normal, texture=/VehicleCosmetics/Bodies/Bawu/Textures/Decals/T_Detail_HexGrille_N}.RGB
   )
   Alpha=@n3
  ) {constA=1}
  n9 = ScalarParameter() {defaultValue=0.5, parameterName=Roughness}
  LinearInterpolate(A=@n9, B=LinearInterpolate(B=@n9, Alpha=@n5.R) {constA=0.5}, Alpha=@n7)
  n10 = Power(Base=@n1.A, Exp=ScalarParameter() {defaultValue=2, parameterName=BorderNormalMultiply})
  n11 = Multiply(A=Saturate(None=Multiply(A=@n10) {constB=2}), B=ScalarParameter() {defaultValue=8, parameterName=BorderNormalInt})
  MaterialFunctionCall!(Height=@n11, Absolute World Position=WorldPosition().XYZ, World Space Vertex Normals=Constant3Vector() {constant=<rgba 0 0 1 0>}) {materialFunction=/Engine/Functions/Engine_MaterialFunctions03/Procedurals/HeightToNormalSmooth}
  n12 = Add(A=@n5.R) {constB=0.33}
  n13 = OneMinus(None=@n5.R)
  Multiply(
   A=VectorParameter() {parameterName=BaseColor, defaultValue=<rgba 0.04 0.04 0.04 0>}.RGB
   B=LinearInterpolate(B=Saturate(None=@n12), Alpha=Multiply(A=@n13, B=@n3)) {constA=1}
  )
  Saturate(None=@n11)
 ]
}
