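# Physics asset for the Bawu vehicle body in the vehicle cosmetics content, providing simplified collision for its skeletal mesh.
# A single body on the root bone holds one box shape, chassis_box, sized to the chassis and offset forward and upward from the bone origin.
# It has no constraints or other bodies, so the asset defines no separate collision or jointed motion for wheels or other parts.
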PhysicsAsset /VehicleCosmetics/Bodies/Bawu/Meshes/Body_Bawu_PhysicsDM
bodies {root [{shapeType=Box, extentX=444, extentY=216, extentZ=116, shapeName=chassis_box, center=<29 0 93>}]}
constraints []
