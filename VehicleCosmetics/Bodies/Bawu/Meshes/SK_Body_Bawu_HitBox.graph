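# Physics asset that gives the Bawu vehicle cosmetic body its hitbox.
# It holds a single box shape on the root bone, offset forward and upward from the bone origin.
# With no other bodies or constraints, the whole body is one simple collision volume with no articulated parts.
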PhysicsAsset /VehicleCosmetics/Bodies/Bawu/Meshes/SK_Body_Bawu_HitBox
bodies {root [{shapeType=Box, extentX=420.7755, extentY=257.6672, extentZ=96.96, shapeName=root_box, center=<40 0 37.6>}]}
constraints []
